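# Translucent, unlit, two-sided Niagara ribbon material for the Ocean Breeze pickaxe swing trail.
# A spline texture borrowed from the Ocean Breeze backpack is sampled on UV1 with scale and pan controls, a randomizable timer and tiling clamp, then desaturated and refined into a mask.
# Opacity combines that mask with particle alpha, a UV0 edge mask and an optional view-angle fade.
# Emissive comes from a ribbon tint multiplied by particle color.
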Material /BRCosmetics/Weapons/FORT_Melee/PIckaxe_OceanBreeze/FX/M_CoeanBreeze_SwingTrail
graph {
 properties {
  bUsedWithNiagaraRibbons=true
  blendMode=BLEND_Translucent
  pixelDepthOffsetMode=PDOM_Legacy
  shadingModel=MSM_Unlit
  twoSided=true
  usedShadingModels=Unlit
 }
 nodes [
  n1 = TextureObjectParameter() {
   group=RibbonA
   parameterName=RibbonA_Texture
   sortPriority=0
   texture=/Game/Accessories/FORT_Backpacks/Backpack_M_MED_OceanBreeze/FX/T_OceanBreeze_SplineA
  }
  n2 = VectorParameter() {group=RibbonA, parameterName=RibbonA_UVControl, channelNames {a=VPan, b=UPan, g=VScale, r=UScale}, defaultValue=<rgba 1.25 1.3 0.2 0>}
  n3 = MaterialFunctionCall!(
   Texture=@n1
   AdditionalUVOffset
   UVControl_RGB=@n2
   UVControl_A=@n2
   "[Override]TimeInput"=NamedRerouteUsage()
   "[Override]ScaleUVByCenter"=StaticBoolParameter() {defaultValue=true, parameterName=ScaleUVByCenter}
   "[Override]UVInput"=TextureCoordinate() {coordinateIndex=1}
  ) {materialFunction=/Game/Effects/Niagara/MasterMaterials/Functions/Common/FXCommon_TextureSample_V2}
  n4 = Multiply(
   A=MaterialFunctionCall!(TexCoord=@n3.UV, ClampChannel=Constant2Vector() {g=1}) {materialFunction=/Game/Effects/Fort_Effects/MaterialFunctions/MF_FX_TilingMethod}.Result
   B=Desaturation(None=@n3.RGB, Fraction=Constant() {r=1})
  )
  n5 = VectorParameter() {parameterName=Refine, defaultValue=<rgba 3 7 2 0.5>}
  n6 = Saturate(None=MaterialFunctionCall(Texture=Saturate(None=@n4), RGB=@n5.RGB, A=@n5.RGB) {materialFunction=&MF_FX_Refine}.Result)
  n7 = ParticleColor()
  n8 = TextureCoordinate()
  n9 = VectorParameter() {group=EdgeB, parameterName=EdgeB_EdgeMask_InvertControl, channelNames {a="", b=Power, g=Multiply, r=Offset}, defaultValue=<rgba 0 3 1 1>}
  n10 = VectorParameter() {group=EdgeB, parameterName=EdgeB_EdgeMask_Control, channelNames {a="", b=Power, g=Multiply, r=Offset}, defaultValue=<rgba 0 8 2 1>}
  n11 = VectorParameter() {parameterName=PostColorMultiply, defaultValue=<rgba 0.5 0.5 0.5 4>}
  n12 = Multiply(
   A=Multiply(
    A=Multiply(A=@n6, B=@n7.A)
    B=MaterialFunctionCall!(UorV=@n8, Invert_EdgeControl=@n9, NoInvert_EdgeControl=@n10) {materialFunction=/Game/Effects/Niagara/MasterMaterials/Functions/Common/FXCommon_EdgeMask_Ex}.Mask
   )
   B=@n11.A
  )
  n13 = ScalarParameter() {defaultValue=1, parameterName=AngleFadeExp}
  LinearInterpolate(
   A=@n12
   B=Multiply(
    A=@n12
    B=MaterialFunctionCall("FadeSensitivity (optional)"=@n13) {materialFunction=&FadeBasedOnViewAngle}.Result
   )
   Alpha=ScalarParameter() {parameterName=UseAngleFade}
  ) {constA=1}
  n14 = Divide(A=VectorParameter() {group=RibbonA, parameterName=RibbonA_Color, defaultValue=<rgba 1.2 2.6 4 0>}.RGB) {constB=1.9771}
  MaterialFunctionCall(Emissive=Multiply(A=Multiply(A=@n14, B=@n7.RGB), B=@n11.RGB)) {materialFunction=&MF_FX_EmissiveAdjustments}
  NamedRerouteDeclaration(None=Add(A=Time() {bOverride_Period=true, period=6666}, B=ScalarParameter() {parameterName=RandomScale})) {name=Timer, nodeColor=<rgba 0 1 0.4844 1>}
 ]
}
