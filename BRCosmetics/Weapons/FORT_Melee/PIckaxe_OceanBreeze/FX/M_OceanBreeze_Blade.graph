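# Additive unlit material for the blade effect of the Ocean Breeze pickaxe, built for Niagara mesh particles.
# An inverted, sharpened Fresnel multiplies a view-angle mask, made by projecting the camera vector off the mesh's local X axis, and particle alpha, with optional depth fade.
# A color parameter, blendable to particle color, feeds an emissive adjustment function.
# A vertex-normal offset is masked along UV V to taper at both ends.
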Material /BRCosmetics/Weapons/FORT_Melee/PIckaxe_OceanBreeze/FX/M_OceanBreeze_Blade
graph {
 properties {
  bUsedWithNiagaraMeshParticles=true
  blendMode=BLEND_Additive
  pixelDepthOffsetMode=PDOM_Legacy
  shadingModel=MSM_Unlit
  usedShadingModels=Unlit
 }
 nodes [
  n1 = OneMinus(None=Fresnel(ExponentIn=ScalarParameter() {defaultValue=1, parameterName=FresnelExp2}) {baseReflectFraction=0})
  n2 = ScalarParameter() {defaultValue=1, parameterName=FresnelExp3}
  n3 = Transform(None=Constant3Vector() {constant=<rgba 1 0 0 0>}) {transformSourceType=TRANSFORMSOURCE_Local}
  n4 = CameraVectorWS()
  n5 = DotProduct(
   A=Normalize(VectorInput=MaterialFunctionCall!(Plane Normal=@n3, Vector=@n4) {materialFunction=/Engine/Functions/Engine_MaterialFunctions02/Math/ProjectVectorOntoPlane}.Result)
   B=PixelNormalWS()
  )
  n6 = Saturate(None=Power(Base=OneMinus(None=@n5), Exp=ScalarParameter() {defaultValue=0.05, parameterName=FresnelExp}))
  n7 = Multiply(A=Multiply(A=Power(Base=@n1, Exp=@n2), B=OneMinus(None=@n6)), B=ParticleColor().A)
  LinearInterpolate(
   A=@n7
   B=DepthFade(Opacity=@n7, FadeDistance=ScalarParameter() {defaultValue=10, parameterName=FadeDistance})
   Alpha=ScalarParameter() {parameterName=Use Depth Fade}
  )
  n8 = VectorParameter() {parameterName=Color, defaultValue=<rgba 0.2 0.4267 1 1>}
  MaterialFunctionCall(Emissive=LinearInterpolate(A=Divide(A=@n8.RGB) {constB=1.9771}, B=ParticleColor().RGB, Alpha=ScalarParameter() {parameterName=Use Particle color})) {materialFunction=&MF_FX_EmissiveAdjustments}
  n9 = VertexNormalWS()
  n10 = ScalarParameter() {defaultValue=10, parameterName=WPOStrength}
  n11 = MaterialFunctionCall(Float2=TextureCoordinate()) {materialFunction=&BreakOutFloat2Components}
  Multiply(A=Multiply(A=Multiply(A=@n9, B=@n10), B=OneMinus(None=@n11.G)), B=Saturate(None=Multiply(A=@n11.G) {constB=3}))
  ScalarParameter() {defaultValue=0.1}
 ]
}
