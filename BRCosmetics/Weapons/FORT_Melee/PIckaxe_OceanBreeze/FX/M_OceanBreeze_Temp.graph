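# Unlit translucent material for Ocean Breeze pickaxe FX, usable on Niagara mesh particles.
# A refined, pannable energy texture multiplies a color blended from base, edge, root and blade tints, feeding emissive adjustments.
# A custom fresnel, using the camera vector projected onto the plane normal to local X, drives edge tint and an opacity term.
# UV edge masks place root and blade tints and shape opacity, with a temporal-dither branch alongside.
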Material /BRCosmetics/Weapons/FORT_Melee/PIckaxe_OceanBreeze/FX/M_OceanBreeze_Temp
graph {
 properties {
  bUsedWithNiagaraMeshParticles=true
  blendMode=BLEND_Translucent
  pixelDepthOffsetMode=PDOM_Legacy
  shadingModel=MSM_Unlit
  usedShadingModels=Unlit
 }
 nodes [
  n1 = VectorParameter() {group=Opacity, parameterName=Opacity_EdgeMask_InvertControl, channelNames {a="", b=Power, g=Multiply, r=Offset}, defaultValue=<rgba 0 1 1 1>}
  n2 = VectorParameter() {group=Opacity, parameterName=Opacity_EdgeMask_Control, channelNames {a="", b=Power, g=Multiply, r=Offset}, defaultValue=<rgba 0 1 1 1>}
  n3 = MaterialFunctionCall!(
   UorV=MaterialFunctionCall(Float2=TextureCoordinate()) {materialFunction=&BreakOutFloat2Components}.G
   Invert_EdgeControl=@n1
   NoInvert_EdgeControl=@n2
  ) {materialFunction=/Game/Effects/Niagara/MasterMaterials/Functions/Common/FXCommon_EdgeMask_Ex}
  n4 = MaterialFunctionCall(Alpha Threshold=@n3.Mask) {materialFunction=&DitherTemporalAA_Fortnite}
  If(A=@n3.Mask, B=Constant() {r=0.99}, "A > B"=Constant() {r=1}, "A == B"=@n4.Result, "A < B"=@n4.Result)
  n5 = VectorParameter() {group=EnergyFlow, parameterName=EnergyFlow_UVControl, channelNames {a=VPan, b=UPan, g=VScale, r=UScale}, defaultValue=<rgba 1 1 0 0>}
  n6 = MaterialFunctionCall!(
   Texture=TextureObjectParameter() {group=EnergyFlow, parameterName=EnergyFlow_Texture, sortPriority=0}
   AdditionalUVOffset
   UVControl_RGB=@n5
   UVControl_A=@n5
   "[Override]TimeInput"
   "[Override]ScaleUVByCenter"
   "[Override]UVInput"
  ) {materialFunction=/Game/Effects/Niagara/MasterMaterials/Functions/Common/FXCommon_TextureSample_V2}
  n7 = VectorParameter() {group=EnergyFlow, parameterName=EnergyFlow_Refine, defaultValue=<rgba 1 1 0 0>}
  n8 = Add(
   A=Multiply(
    A=MaterialFunctionCall(Texture=@n6.R, RGB=@n7.RGB, A=@n7.RGB) {materialFunction=&MF_FX_Refine}.Result
    B=ScalarParameter() {parameterName=EnergyFlowStr}
   )
   B=ScalarParameter() {parameterName=EnergyFlowOffset}
  )
  n9 = VectorParameter() {parameterName=BaseC, defaultValue=<rgba 0 0.6416 1 1>}
  n10 = VectorParameter() {parameterName=EdgeC, defaultValue=<rgba 0 0.6416 1 1>}
  n11 = Transform(None=Constant3Vector() {constant=<rgba 1 0 0 0>}) {transformSourceType=TRANSFORMSOURCE_Local}
  n12 = CameraVectorWS()
  n13 = DotProduct(
   A=Normalize(VectorInput=MaterialFunctionCall!(Plane Normal=@n11, Vector=@n12) {materialFunction=/Engine/Functions/Engine_MaterialFunctions02/Math/ProjectVectorOntoPlane}.Result)
   B=PixelNormalWS()
  )
  n14 = ScalarParameter() {defaultValue=2, parameterName=FresnelExp}
  n15 = Power(Base=OneMinus(None=@n13), Exp=@n14)
  n16 = MaterialFunctionCall(Float2=TextureCoordinate()) {materialFunction=&BreakOutFloat2Components}
  n17 = VectorParameter() {
   group=RootMask
   parameterName=RootMask_EdgeMask_InvertControl
   channelNames {a="", b=Power, g=Multiply, r=Offset}
   defaultValue=<rgba 0 1 1 1>
  }
  n18 = VectorParameter() {group=RootMask, parameterName=RootMask_EdgeMask_Control, channelNames {a="", b=Power, g=Multiply, r=Offset}, defaultValue=<rgba 0 1 1 1>}
  n19 = VectorParameter() {
   group=BladeMask
   parameterName=BladeMask_EdgeMask_InvertControl
   channelNames {a="", b=Power, g=Multiply, r=Offset}
   defaultValue=<rgba 0 1 1 1>
  }
  n20 = VectorParameter() {group=BladeMask, parameterName=BladeMask_EdgeMask_Control, channelNames {a="", b=Power, g=Multiply, r=Offset}, defaultValue=<rgba 0 1 1 1>}
  n21 = LinearInterpolate(
   A=LinearInterpolate(
    A=LinearInterpolate(A=@n9.RGB, B=@n10.RGB, Alpha=@n15)
    B=VectorParameter() {parameterName=RootC, defaultValue=<rgba 0 0.6416 1 1>}.RGB
    Alpha=MaterialFunctionCall(UorV=@n16.G, Invert_EdgeControl=@n17, NoInvert_EdgeControl=@n18) {materialFunction=&FXCommon_EdgeMask_Ex}.Mask
   )
   B=VectorParameter() {parameterName=BladeC, defaultValue=<rgba 0 0.6416 1 1>}.RGB
   Alpha=OneMinus(None=MaterialFunctionCall(UorV=@n16.R, Invert_EdgeControl=@n19, NoInvert_EdgeControl=@n20) {materialFunction=&FXCommon_EdgeMask_Ex}.Mask)
  )
  MaterialFunctionCall(Emissive=Divide(A=Multiply(A=@n8, B=@n21)) {constB=1.9771}) {materialFunction=&MF_FX_EmissiveAdjustments}
  Max(B=Fresnel(ExponentIn=@n14) {baseReflectFraction=0})
  n22 = OneMinus(None=Power(Base=@n13, Exp=ScalarParameter() {defaultValue=5, parameterName=Opacity_Fresnel2}))
  n23 = ScalarParameter() {defaultValue=5, parameterName=Opacity_Fresnel}
  Multiply(A=@n3.Mask, B=OneMinus(None=Power(Base=@n22, Exp=@n23)))
 ]
}
