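# Unlit translucent material for Niagara ribbons, named as the tip ribbon effect of the OceanBreeze pickaxe.
# Particle color is multiplied by the ColorM tint and passed through the shared emissive adjustments function.
# A texture-free alpha mask multiplies particle alpha by two FXCommon edge masks, Edge on U and Tip on V, each shaped by offset, multiply and power parameters.
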Material /BRCosmetics/Weapons/FORT_Melee/PIckaxe_OceanBreeze/FX/M_OceanBreeze_TipRibbon
graph {
 properties {bUsedWithNiagaraRibbons=true, blendMode=BLEND_Translucent, pixelDepthOffsetMode=PDOM_Legacy, shadingModel=MSM_Unlit, usedShadingModels=Unlit}
 nodes [
  n1 = ParticleColor()
  MaterialFunctionCall(Emissive=Multiply(A=@n1.RGB, B=VectorParameter() {parameterName=ColorM, defaultValue=<rgba 2 3 4 1>}.RGB)) {materialFunction=&MF_FX_EmissiveAdjustments}
  n2 = MaterialFunctionCall(Float2=TextureCoordinate()) {materialFunction=&BreakOutFloat2Components}
  n3 = VectorParameter() {group=Edge, parameterName=Edge_EdgeMask_InvertControl, channelNames {a="", b=Power, g=Multiply, r=Offset}, defaultValue=<rgba 0 2 3 1>}
  n4 = VectorParameter() {group=Edge, parameterName=Edge_EdgeMask_Control, channelNames {a="", b=Power, g=Multiply, r=Offset}, defaultValue=<rgba 0 3 2 1>}
  n5 = VectorParameter() {group=Tip, parameterName=Tip_EdgeMask_InvertControl, channelNames {a="", b=Power, g=Multiply, r=Offset}, defaultValue=<rgba 0 2 100 1>}
  n6 = VectorParameter() {group=Tip, parameterName=Tip_EdgeMask_Control, channelNames {a="", b=Power, g=Multiply, r=Offset}, defaultValue=<rgba 0 2 30 1>}
  Multiply(
   A=MaterialFunctionCall!(UorV=@n2.R, Invert_EdgeControl=@n3, NoInvert_EdgeControl=@n4) {materialFunction=/Game/Effects/Niagara/MasterMaterials/Functions/Common/FXCommon_EdgeMask_Ex}.Mask
   B=Multiply(
    A=@n1.A
    B=MaterialFunctionCall(UorV=@n2.G, Invert_EdgeControl=@n5, NoInvert_EdgeControl=@n6) {materialFunction=&FXCommon_EdgeMask_Ex}.Mask
   )
  )
 ]
}
