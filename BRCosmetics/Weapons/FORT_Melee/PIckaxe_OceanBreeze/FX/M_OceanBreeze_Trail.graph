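# Translucent unlit material for the Niagara ribbon trail of the Ocean Breeze pickaxe, with refraction and a rainbow tint.
# A sine pattern on the second UV set is remapped, contrast-adjusted and multiplied by an edge mask to form the shape mask.
# That mask drives refraction and, with particle alpha and a second edge mask, opacity.
# Color adds particle-tinted base color to a desaturation-adjusted rainbow texture that strengthens as particle alpha fades.
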Material /BRCosmetics/Weapons/FORT_Melee/PIckaxe_OceanBreeze/FX/M_OceanBreeze_Trail
graph {
 properties {
  bUsedWithNiagaraRibbons=true
  blendMode=BLEND_Translucent
  pixelDepthOffsetMode=PDOM_Legacy
  refractionMethod=RM_IndexOfRefraction
  shadingModel=MSM_Unlit
  usedShadingModels=Unlit
 }
 nodes [
  n1 = MaterialFunctionCall(Float2=TextureCoordinate()) {materialFunction=&BreakOutFloat2Components}
  n2 = VectorParameter() {
   group=OpacityMask
   parameterName=OpacityMask_EdgeMask_InvertControl
   channelNames {a="", b=Power, g=Multiply, r=Offset}
   defaultValue=<rgba 0 1 1 1>
  }
  n3 = VectorParameter() {
   group=OpacityMask
   parameterName=OpacityMask_EdgeMask_Control
   channelNames {a="", b=Power, g=Multiply, r=Offset}
   defaultValue=<rgba 0 10 2 1>
  }
  n4 = VectorParameter() {parameterName=GridUV, defaultValue=<rgba 10 0.05 0 1>}
  n5 = Multiply(A=TextureCoordinate() {coordinateIndex=1}, B=AppendVector(A=@n4.R, B=@n4.R))
  n6 = Sine(None=MaterialFunctionCall(Float2=Add(A=@n5) {constB=0.01}) {materialFunction=&BreakOutFloat2Components}.R)
  n7 = VectorParameter() {parameterName=GridRemap, defaultValue=<rgba -0.8 1 1 0.5>}
  MaterialExpressionRemap_0 = MaterialXRemap!(None=@n6, InputLow=@n7.R, InputHigh=@n7.R, TargetLow=@n7.R, TargetHigh=@n7.R)
  n8 = VectorParameter() {parameterName=GridOpacityOffset, defaultValue=<rgba 1.4 0.7 0 1>}
  n9 = Saturate(None=MaterialFunctionCall(In=Add(A=@n6, B=@n8.R), Contrast=ScalarParameter() {defaultValue=0.8, parameterName=GridContrast}) {materialFunction=&CheapContrast}.Result)
  n10 = VectorParameter() {parameterName=BaseLayerColor, defaultValue=<rgba 0.35 0.425 0.5 0.7>}
  n11 = Multiply(
   A=Saturate(None=MaterialFunctionCall!(UorV=@n1.G, Invert_EdgeControl=@n2, NoInvert_EdgeControl=@n3) {materialFunction=/Game/Effects/Niagara/MasterMaterials/Functions/Common/FXCommon_EdgeMask_Ex}.Mask)
   B=Multiply(A=Multiply(A=@MaterialExpressionRemap_0, B=@n9) {constA=1}, B=@n10.A)
  )
  n12 = Constant()
  n13 = ParticleColor()
  LinearInterpolate(
   B=ScalarParameter() {defaultValue=1.1, parameterName=Refraction}
   Alpha=LinearInterpolate(B=If(A=@n11, "A > B"=@n11, "A == B"=@n11, "A < B"=@n12) {constB=0.01}, Alpha=Power(Base=@n13.A) {constExponent=1.3})
  ) {constA=1}
  n14 = OneMinus(None=@n13.A)
  n15 = TextureObjectParameter() {group=RainbowColor, parameterName=RainbowColor_Texture, sortPriority=0, texture=&T_Tile_RainbowColor}
  n16 = VectorParameter() {group=RainbowColor, parameterName=RainbowColor_UVControl, channelNames {a=VPan, b=UPan, g=VScale, r=UScale}, defaultValue=<rgba 3 0.3 0 0>}
  n17 = Desaturation(
   None=MaterialFunctionCall(Texture=@n15, UVControl_RGB=@n16, UVControl_A=@n16) {materialFunction=&FXCommon_TextureSample}.RGB
   Fraction=ScalarParameter() {defaultValue=-2, parameterName=RainbowDes}
  )
  n18 = ScalarParameter() {defaultValue=2.3, parameterName=RainbowStrength}
  n19 = Add(A=Multiply(A=Power(Base=@n14) {constExponent=1.5}, B=Multiply(A=@n17, B=@n18)), B=Multiply(A=@n13.RGB, B=@n10.RGB))
  n20 = VectorParameter() {parameterName=PostColorMulti, defaultValue=<rgba 0.5 0.5 0.5 2>}
  MaterialFunctionCall(Emissive=Divide(A=Multiply(A=@n19, B=@n20.RGB)) {constB=1.9771}) {materialFunction=&MF_FX_EmissiveAdjustments}
  MaterialFunctionCall(Alpha Threshold=@n11, Random=Constant()) {materialFunction=&DitherTemporalAA_Fortnite}
  n21 = TextureCoordinate()
  n22 = VectorParameter() {group=EdgeB, parameterName=EdgeB_EdgeMask_InvertControl, channelNames {a="", b=Power, g=Multiply, r=Offset}, defaultValue=<rgba 0 3 1.5 1>}
  n23 = VectorParameter() {group=EdgeB, parameterName=EdgeB_EdgeMask_Control, channelNames {a="", b=Power, g=Multiply, r=Offset}, defaultValue=<rgba 0 5 2 1>}
  Multiply(
   A=Multiply(
    A=Multiply(A=@n11, B=@n13.A)
    B=MaterialFunctionCall(UorV=@n21, Invert_EdgeControl=@n22, NoInvert_EdgeControl=@n23) {materialFunction=&FXCommon_EdgeMask_Ex}.Mask
   )
   B=@n20.A
  )
  TextureCoordinate()
 ]
}
