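# Translucent unlit material for Niagara ribbons, named as the trail center of the OceanBreeze pickaxe.
# U and V edge-mask functions, with a tip falloff on V, build one procedural gradient without textures.
# The reshaped gradient blends EdgeColor to CenterColor for emissive, scaled down and passed through MF_FX_EmissiveAdjustments.
# The same pre-scaled gradient, after contrast, power and multiply, is multiplied by particle alpha for opacity.
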Material /BRCosmetics/Weapons/FORT_Melee/PIckaxe_OceanBreeze/FX/M_OceanBreeze_TrailCenter
graph {
 properties {bUsedWithNiagaraRibbons=true, blendMode=BLEND_Translucent, pixelDepthOffsetMode=PDOM_Legacy, shadingModel=MSM_Unlit, usedShadingModels=Unlit}
 nodes [
  n1 = VectorParameter() {parameterName=EdgeColor, defaultValue=<rgba 0.2 0.8 2 1>}
  n2 = VectorParameter() {parameterName=CenterColor, defaultValue=<rgba 0.09 0.1 0.15 1>}
  n3 = MaterialFunctionCall(Float2=TextureCoordinate()) {materialFunction=&BreakOutFloat2Components}
  n4 = VectorParameter() {group=EdgeU, parameterName=EdgeU_EdgeMask_InvertControl, channelNames {a="", b=Power, g=Multiply, r=Offset}, defaultValue=<rgba 0 1 1.2 1>}
  n5 = VectorParameter() {group=EdgeU, parameterName=EdgeU_EdgeMask_Control, channelNames {a="", b=Power, g=Multiply, r=Offset}, defaultValue=<rgba 0 1.3 1 1>}
  n6 = VectorParameter() {group=EdgeV, parameterName=EdgeV_EdgeMask_InvertControl, channelNames {a="", b=Power, g=Multiply, r=Offset}, defaultValue=<rgba 0 2 1 1>}
  n7 = VectorParameter() {group=EdgeV, parameterName=EdgeV_EdgeMask_Control, channelNames {a="", b=Power, g=Multiply, r=Offset}, defaultValue=<rgba 0 2 1 1>}
  n8 = Power(
   Base=OneMinus(None=MaterialFunctionCall!(UorV=@n3.G, Invert_EdgeControl=@n6, NoInvert_EdgeControl=@n7) {materialFunction=/Game/Effects/Niagara/MasterMaterials/Functions/Common/FXCommon_EdgeMask_Ex}.Mask)
   Exp=ScalarParameter() {defaultValue=0.5, parameterName=GradientTipScale}
  )
  n9 = Multiply(
   A=Multiply(
    A=MaterialFunctionCall(UorV=@n3.R, Invert_EdgeControl=@n4, NoInvert_EdgeControl=@n5) {materialFunction=&FXCommon_EdgeMask_Ex}.Mask
    B=OneMinus(None=@n8)
   )
   B=ScalarParameter() {defaultValue=4, parameterName=GradientPreScale}
  )
  n10 = Power(Base=@n9, Exp=ScalarParameter() {defaultValue=0.7, parameterName=GradientPowerA})
  n11 = OneMinus(None=Power(Base=OneMinus(None=@n10), Exp=ScalarParameter() {defaultValue=1.6, parameterName=GradientPowerB}))
  MaterialFunctionCall(Emissive=Divide(A=LinearInterpolate(A=@n1.RGB, B=@n2.RGB, Alpha=@n11)) {constB=1.9771}) {materialFunction=&MF_FX_EmissiveAdjustments}
  n12 = MaterialFunctionCall(In=@n9, Contrast=ScalarParameter() {defaultValue=0.5, parameterName=OpacityContrast}) {materialFunction=&CheapContrast}
  n13 = ScalarParameter() {defaultValue=3, parameterName=OpacityPower}
  Multiply(A=ParticleColor().A, B=Multiply(A=Power(Base=@n12.Result, Exp=@n13), B=ScalarParameter() {defaultValue=0.5, parameterName=OpacityMultiply}))
 ]
}
