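# Niagara swing trail for the Ocean Breeze pickaxe, built from short ribbon streaks and sparkles in white-to-blue tones.
# Four anim-trail ribbons follow the User skeletal mesh, layering different widths, lifetimes and colour/alpha curves.
# A wide ribbon, dimmed by time of day, spawns at a point lerped between two system-level mesh samples.
# Distance-spawned sprites inherit mesh velocity, then drift with curl noise and drag through two sprite renderers.
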NiagaraSystem /BRCosmetics/Weapons/FORT_Melee/PIckaxe_OceanBreeze/FX/NS_Pickaxe_OceanBreeze_Trail
dependencies {
 usedDataInterfaces [&NiagaraDataInterfaceCurve, &NiagaraDataInterfaceColorCurve]
 usedDynamicInputs []
 usedModules [&SystemState, &SampleSkeletalMesh]
}
emitters {
 Empty {
  emitterUpdateScript [
   &EmitterState(Life Cycle Mode=::NewEnumerator1, Loop Behavior=::NewEnumerator1, Loop Duration=0.2, UseLoopDelay=false)
   &SpawnPerUnit(Module.Use Max Movement Threshold=true, Module.Use Movement Tolerance=false, Spawn Spacing=25, Use Spawn Probability=false)
  ]
  particleSpawnScript [
   &InitializeParticle(
    Color Maximum=<rgba 3.5 4.25 5 1>
    Color Minimum=<rgba 0.14 0.17 0.2 1>
    Color Mode=::NewEnumerator2
    Lifetime Max=0.5
    Lifetime Min=0.1
    Lifetime Mode=::NewEnumerator1
    Sprite Size Mode=::NewEnumerator1
    Uniform Sprite Size Max=1
    Uniform Sprite Size Min=0.2
    UsePositionOffset=false
    Write Material Random=true
   )
   &SkeletalMeshLocation(MaskSpawnGroup=false, Mesh Sampling Type=::NewEnumerator5, Skeletal Mesh=$User.SkeletalMesh, Socket Sampling Mode=::NewEnumerator5)
   &InheritVelocity(
    Inherited Velocity Speed Limit:NiagaraFloat=&RandomRangeFloat(Minimum=40, Maximum=80)
    Speed Limit=true
    UseSpeedThreshold=false
   )
   &ApplyInitialForces()
   &ShapeLocation(MaskSpawnGroup=false, Override Seed=false, Sphere Radius=10, UseOwnerScale=true)
  ]
  particleUpdateScript [
   &ParticleState()
   &ScaleSpriteSize(Uniform Curve Sprite Scale=DataInterface() {propertyValues {Curve {keys [
        {tangentMode=RCTM_User, time=0, value=1}
        {arriveTangent=-0.0405, interpMode=RCIM_Cubic, leaveTangent=-0.0405, tangentMode=RCTM_User, time=0.2, value=1}
        {arriveTangent=-2.1547, interpMode=RCIM_Cubic, leaveTangent=-2.1547, tangentMode=RCTM_User, time=1, value=0}
       ]}}})
   &CurlNoiseForce(Noise Frequency=20, Noise Strength=200, Pan Noise=false)
   &Drag(Drag=&RandomRangeFloat(Minimum=4, Maximum=8), Use Linear Drag=true, Use Rotational Drag=true)
   &SolveForcesAndVelocity(Clamp Velocity=false, Limit Acceleration=false)
   /SetVariables(
    Particles.Color2=&MultiplyLinearColor(
     Linear Color=&RandomRangeLinearColor(Minimum=<rgba 0.005 0.0077 0.01 1>, Maximum=<rgba 0 0.035 0.07 1>)
     Scale Factor=&MakeVector4(
      X=1
      Y=1
      Z=1
      W=&FloatFromCurve(FloatCurve=DataInterface() {propertyValues {Curve {keys [
           {interpMode=RCIM_Cubic, tangentMode=RCTM_User, time=0, value=0}
           {interpMode=RCIM_Cubic, tangentMode=RCTM_User, time=0.5, value=1}
           {tangentMode=RCTM_User, time=1, value=0}
          ]}}})
     )
    )
    Particles.SpriteSize2=&Multiply_Vector2D_ByFloat(
     Float=&Waveform(
      Global Phase=$Particles.NormalizedAge
      "Waveform [1]"=::NewEnumerator0
      "[1] Phase Scale"=3
      "[1] Frequency"=1
      "[1] Amplitude Min/Max"=<0.5 1>
      "[1] Bias"=1
     )
     Vector2D=<5 5>
    )
   ) {isSetParameters=true}
  ]
  rendererClasses [&NiagaraSpriteRendererProperties]
  renderers [
   {rendererClass=&NiagaraSpriteRendererProperties}
   {rendererClass=&NiagaraSpriteRendererProperties, rendererIndex=1}
  ]
 }
 FN_AnimTrail_Short {
  emitterUpdateScript [
   &EmitterState(Life Cycle Mode=::NewEnumerator1, Loop Behavior=::NewEnumerator1, Loop Duration=0.15, UseLoopDelay=false)
   &FN_SpawnAnimTrail(AnimTrailWidthMode=::FromFirst)
  ]
  particleSpawnScript [
   &FN_InitializeAnimTrailBoneIndecies()
   &FN_InitializeAnimTrail(AnimTrailWidth=0.5, Color=<rgba 0.7 0.7 0.7 1>, Lifetime=0.3)
   /SetVariables(Particles.Position=&Lerp_Vector(A=$System.SampleSkeletalMesh.SampledPosition, B=$System.SampleSkeletalMesh001.SampledPosition, Alpha=0.4)) {isSetParameters=true}
  ]
  particleUpdateScript [
   &SolveForcesAndVelocity(Clamp Velocity=false, Limit Acceleration=false)
   &FN_UpdateAnimTrail(SkeletalMesh=$User.SkeletalMesh) {module=FN_UpdateAnimTrail001}
   &ScaleColor(
    Scale Alpha=&FloatFromCurve(FloatCurve=DataInterface() {propertyValues {Curve {keys [
         {interpMode=RCIM_Cubic, leaveTangent=0.2469, tangentMode=RCTM_Break, time=0, value=0}
         {arriveTangent=-0.1054, interpMode=RCIM_Cubic, leaveTangent=-9.7681, tangentMode=RCTM_Break, time=0.1175, value=0.9999}
         {arriveTangent=-0.5795, interpMode=RCIM_Cubic, leaveTangent=-0.5795, tangentMode=RCTM_User, time=0.2182, value=0.1701}
         <at 1 0>
        ]}}})
    Scale RGB=&MakeVectorFromLinearColorRGB(Color=&ColorFromCurve(ColorCurve=DataInterface() {propertyValues {
        AlphaCurve {keys [<at 0 1 RCIM_Cubic>,<at 1 0 RCIM_Cubic>]}
        BlueCurve {keys [<at 0 10 RCIM_Cubic>,<at 0.1322 50>,<at 0.2868 1>,<at 0.7159 1 RCIM_Cubic>]}
        GreenCurve {keys [<at 0 7.5 RCIM_Cubic>,<at 0.1322 32.5>,<at 0.2868 0.75>,<at 0.7159 0.5 RCIM_Cubic>]}
        RedCurve {keys [<at 0 5 RCIM_Cubic>,<at 0.1322 15>,<at 0.2868 0.5>,<at 0.7159 0 RCIM_Cubic>]}
       }}))
    ScaleA=true
    ScaleRGB=true
   )
   &ScaleRibbonWidth(Ribbon Width Scale=&FloatFromCurve(FloatCurve=DataInterface() {propertyValues {Curve {keys [
         {interpMode=RCIM_Cubic, leaveTangent=-1.8911, tangentMode=RCTM_Break, time=0, value=1}
         {arriveTangent=-0.0883, interpMode=RCIM_Cubic, leaveTangent=-0.0883, tangentMode=RCTM_User, time=1, value=0}
        ]}}}))
  ]
  rendererClasses [&NiagaraRibbonRendererProperties]
  renderers [{rendererClass=&NiagaraRibbonRendererProperties}]
 }
 FN_AnimTrail_Short004 {
  emitterUpdateScript [
   &EmitterState(Life Cycle Mode=::NewEnumerator1, Loop Behavior=::NewEnumerator1, Loop Duration=0.2, UseLoopDelay=false)
   &FN_SpawnAnimTrail()
  ]
  particleSpawnScript [
   &FN_InitializeAnimTrailBoneIndecies()
   &FN_InitializeAnimTrail(AnimTrailWidth=1.4, Lifetime=0.25)
  ]
  particleUpdateScript [
   &SolveForcesAndVelocity(Clamp Velocity=false, Limit Acceleration=false)
   &FN_UpdateAnimTrail(SkeletalMesh=$User.SkeletalMesh) {module=FN_UpdateAnimTrail001}
   &ScaleColor(
    Scale Alpha=&FloatFromCurve(FloatCurve=DataInterface() {propertyValues {Curve {keys [
         {interpMode=RCIM_Cubic, leaveTangent=33.6429, tangentMode=RCTM_Break, time=0, value=0}
         {interpMode=RCIM_Cubic, leaveTangent=-0.9382, tangentMode=RCTM_Break, time=0.0739, value=0.9931}
         {arriveTangent=-0.5568, interpMode=RCIM_Cubic, tangentMode=RCTM_Break, time=1, value=0}
        ]}}})
    Scale RGB=&MakeVectorFromLinearColorRGB(Color=&ColorFromCurve(ColorCurve=DataInterface() {propertyValues {
        AlphaCurve {keys [<at 0 1 RCIM_Cubic>,<at 1 0 RCIM_Cubic>]}
        BlueCurve {keys [{arriveTangent=-13.2888, interpMode=RCIM_Cubic, time=0.122, value=1}, <at 0.3559 1 RCIM_Cubic>, <at 0.5573 1>, <at 0.6956 1>]}
        GreenCurve {keys [
          {arriveTangent=-14.1664, interpMode=RCIM_Cubic, time=0.122, value=1}
          {arriveTangent=-0.6198, interpMode=RCIM_Cubic, leaveTangent=-0.6198, time=0.3559, value=0.7303}
          <at 0.5573 0.7303>
          <at 0.6956 0.55>
         ]}
        RedCurve {keys [
          {arriveTangent=-14.7653, interpMode=RCIM_Cubic, time=0.122, value=1}
          {arriveTangent=-1.6083, interpMode=RCIM_Cubic, leaveTangent=-1.6083, time=0.3559, value=0.3}
          <at 0.5573 0.3>
          <at 0.6956 0>
         ]}
       }}))
    ScaleA=true
    ScaleRGB=true
   )
   &ScaleRibbonWidth(Ribbon Width Scale=&FloatFromCurve(FloatCurve=DataInterface() {propertyValues {Curve {keys [
         {interpMode=RCIM_Cubic, leaveTangent=-0.0002, tangentMode=RCTM_Break, time=0.2, value=1}
         {interpMode=RCIM_Cubic, leaveTangent=-3.3704, time=1, value=1}
        ]}}}))
  ]
  rendererClasses [&NiagaraRibbonRendererProperties]
  renderers [{rendererClass=&NiagaraRibbonRendererProperties}]
 }
 FN_AnimTrail_Short005 {
  emitterUpdateScript [
   &EmitterState(Life Cycle Mode=::NewEnumerator1, Loop Behavior=::NewEnumerator1, Loop Duration=0.2, UseLoopDelay=false)
   &FN_SpawnAnimTrail()
  ]
  particleSpawnScript [
   &FN_InitializeAnimTrailBoneIndecies()
   &FN_InitializeAnimTrail(AnimTrailWidth=1.25, Lifetime=0.19)
   /SetVariables(Particles.Position=&Lerp_Vector(A=$System.SampleSkeletalMesh.SampledPosition, B=$System.SampleSkeletalMesh001.SampledPosition, Alpha=0.4)) {isSetParameters=true}
  ]
  particleUpdateScript [
   &SolveForcesAndVelocity(Clamp Velocity=false, Limit Acceleration=false)
   &FN_UpdateAnimTrail(SkeletalMesh=$User.SkeletalMesh) {module=FN_UpdateAnimTrail001}
   &ScaleColor(
    Scale Alpha=&FloatFromCurve(FloatCurve=DataInterface() {propertyValues {Curve {keys [
         {interpMode=RCIM_Cubic, leaveTangent=-0.3092, tangentMode=RCTM_Break, time=0.0485, value=1}
         {arriveTangent=-4.82, interpMode=RCIM_Cubic, leaveTangent=-0.1565, tangentMode=RCTM_Break, time=0.117, value=0.2896}
         <at 1 0>
        ]}}})
    Scale RGB=&MakeVectorFromLinearColorRGB(Color=&ColorFromCurve(ColorCurve=DataInterface() {propertyValues {
        AlphaCurve {keys [<at 0 1 RCIM_Cubic>,<at 1 0 RCIM_Cubic>]}
        BlueCurve {keys [{arriveTangent=-13.2888, interpMode=RCIM_Cubic, time=0.5105, value=1}, <at 0.604 1 RCIM_Cubic>]}
        GreenCurve {keys [{arriveTangent=-14.1664, interpMode=RCIM_Cubic, time=0.5105, value=1}, <at 0.604 0.4056 RCIM_Cubic>]}
        RedCurve {keys [{arriveTangent=-14.7653, interpMode=RCIM_Cubic, time=0.5105, value=1}, <at 0.604 0 RCIM_Cubic>]}
       }}))
    ScaleA=true
    ScaleRGB=true
   )
   &ScaleRibbonWidth(Ribbon Width Scale=&FloatFromCurve(FloatCurve=DataInterface() {propertyValues {Curve {keys [
         {interpMode=RCIM_Cubic, leaveTangent=-0.1469, tangentMode=RCTM_Break, time=0.4, value=1}
         {arriveTangent=-3.3704, interpMode=RCIM_Cubic, leaveTangent=-3.3704, tangentMode=RCTM_User, time=1, value=0}
        ]}}}))
  ]
  rendererClasses [&NiagaraRibbonRendererProperties]
  renderers [{rendererClass=&NiagaraRibbonRendererProperties}]
 }
 FN_AnimTrail_Short006 {
  emitterUpdateScript [
   &EmitterState(Life Cycle Mode=::NewEnumerator1, Loop Behavior=::NewEnumerator1, Loop Duration=0.25, UseLoopDelay=false)
   &FN_SpawnAnimTrail(Spawn Interval=0)
  ]
  particleSpawnScript [
   &FN_InitializeAnimTrailBoneIndecies()
   &FN_InitializeAnimTrail(AnimTrailWidth=0.7, Color=<rgba 1 1 1 1.2>, Lifetime=0.6)
  ]
  particleUpdateScript [
   &SolveForcesAndVelocity(Clamp Velocity=false, Limit Acceleration=false)
   &FN_UpdateAnimTrail(SkeletalMesh=$User.SkeletalMesh) {module=FN_UpdateAnimTrail001}
   &ScaleColor(
    Scale Alpha=&FloatFromCurve(
     FloatCurve=DataInterface() {propertyValues {Curve {keys [
         {interpMode=RCIM_Cubic, leaveTangent=33.6429, tangentMode=RCTM_Break, time=0, value=0}
         {interpMode=RCIM_Cubic, leaveTangent=-0.3092, tangentMode=RCTM_Break, time=0.0739, value=0.9931}
         {arriveTangent=-4.82, interpMode=RCIM_Cubic, leaveTangent=-0.1565, tangentMode=RCTM_Break, time=0.3, value=0.25}
         <at 1 0>
        ]}}}
     CurveIndex=&Multiply_Float(A=$Particles.Age, B=3.9)
    )
    Scale RGB=&MakeVectorFromLinearColorRGB(Color=&ColorFromCurve(
      ColorCurve=DataInterface() {propertyValues {
        AlphaCurve {keys [<at 0 1 RCIM_Cubic>,<at 1 0 RCIM_Cubic>]}
        BlueCurve {keys [{arriveTangent=-13.2888, time=0.061, value=10}, {arriveTangent=-24.8604, time=0.303, value=1.2}, <at 0.4881 1>, <at 0.7281 1>, <at 1 1>]}
        GreenCurve {keys [
          {arriveTangent=-14.1664, time=0.061, value=7.8333}
          {arriveTangent=-25.6056, time=0.303, value=0.7528}
          <at 0.4881 0.6013>
          <at 0.7281 0.55>
          <at 1 0.6167>
         ]}
        RedCurve {keys [{arriveTangent=-14.7653, time=0.061, value=5}, {arriveTangent=-26.794, time=0.303, value=0.168}, <at 0.4881 0.08>, <at 0.7281 0>, <at 1 0>]}
       }}
      CurveIndex=&Multiply_Float(A=$Particles.Age, B=4)
     ))
    ScaleA=true
    ScaleRGB=true
   )
   &ScaleColor(
    Scale Alpha=&FloatFromCurve(
     FloatCurve=DataInterface() {propertyValues {Curve {keys [
         {interpMode=RCIM_Cubic, leaveTangent=33.6429, tangentMode=RCTM_Break, time=0, value=0}
         {interpMode=RCIM_Cubic, leaveTangent=-0.3092, tangentMode=RCTM_Break, time=0.0739, value=0.9931}
         {arriveTangent=-4.82, interpMode=RCIM_Cubic, leaveTangent=-0.1565, tangentMode=RCTM_Break, time=0.3, value=0.25}
         <at 1 0>
        ]}}}
     CurveIndex=&Multiply_Float(A=$Particles.Age, B=3.9)
    )
    Scale RGB=&MakeVectorFromLinearColorRGB(Color=&ColorFromCurve(
      ColorCurve=DataInterface() {propertyValues {
        AlphaCurve {keys [<at 0 1 RCIM_Cubic>,<at 1 0 RCIM_Cubic>]}
        BlueCurve {keys [{arriveTangent=-13.2888, time=0.061, value=10}, {arriveTangent=-24.8604, time=0.2888, value=1.2}, <at 0.5288 1>, <at 0.7403 1>, <at 1 1>]}
        GreenCurve {keys [
          {arriveTangent=-14.1664, time=0.061, value=7.8333}
          {arriveTangent=-25.6056, time=0.2888, value=0.7528}
          <at 0.5288 0.6013>
          <at 0.7403 0.55>
          <at 1 0.6167>
         ]}
        RedCurve {keys [{arriveTangent=-14.7653, time=0.061, value=5}, {arriveTangent=-26.794, time=0.2888, value=0.168}, <at 0.5288 0.08>, <at 0.7403 0>, <at 1 0>]}
       }}
      CurveIndex=&Multiply_Float(A=$Particles.Age, B=4)
     ))
    ScaleA=true
    ScaleRGB=true
   ) {enabled=false, module=ScaleColor002}
   &ScaleRibbonWidth(Ribbon Width Scale=&FloatFromCurve(
     FloatCurve=DataInterface() {propertyValues {Curve {keys [
         {interpMode=RCIM_Cubic, leaveTangent=-0.0001, tangentMode=RCTM_Break, time=0.2, value=1}
         {interpMode=RCIM_Cubic, leaveTangent=-3.3704, time=1, value=1.7}
        ]}}}
     CurveIndex=&Multiply_Float(A=$Particles.Age, B=4.75)
    ))
   &ScaleColor(Scale Alpha=0.3, ScaleA=true, ScaleRGB=true) {module=ScaleColor001}
  ]
  rendererClasses [&NiagaraRibbonRendererProperties]
  renderers [{rendererClass=&NiagaraRibbonRendererProperties}]
 }
 Trail_color {
  emitterUpdateScript [
   &EmitterState(Life Cycle Mode=::NewEnumerator1, Loop Behavior=::NewEnumerator1, Loop Duration=0.2, UseLoopDelay=false)
   &SpawnRate(SpawnRate=30, Use Spawn Probability=false)
  ]
  particleSpawnScript [
   &InitializeRibbon(
    Color=<rgba 1 1 1 0.75>
    Lifetime=0.25
    Module.Write Color=true
    Module.Write Lifetime=true
    Module.Write Mass=true
    Module.Write Material Random=true
    Module.Write Position=false
    Module.Write RibbonTwist=false
    Module.Write RibbonWidth=true
    Ribbon Width=28
   )
   /SetVariables(Particles.Position=&Lerp_Vector(A=$System.SampleSkeletalMesh.SampledPosition, B=$System.SampleSkeletalMesh001.SampledPosition, Alpha=0.35)) {isSetParameters=true}
  ]
  particleUpdateScript [
   &ParticleState()
   &SolveForcesAndVelocity(Clamp Velocity=false, Limit Acceleration=false)
   &ScaleColor(
    Scale Alpha=&Multiply_Float(
     A=&Multiply_Float(A=1, B=&FNTimeOfDayScalar(Night Scale=0.5, Morning Scale=0.5, Evening Scale=0.5))
     B=&FloatFromCurve(FloatCurve=DataInterface() {propertyValues {Curve {keys [
          <at 0 1 RCIM_Cubic>
          {arriveTangent=-2.0952, interpMode=RCIM_Cubic, leaveTangent=-2.0952, time=0.18, value=0.4}
          <at 0.42 0.12>
          <at 0.7723 0.0471>
          <at 0.7898 -0.0042>
          <at 1 0>
         ]}}})
    )
    Scale RGB=&MakeVectorFromLinearColorRGB(Color=<rgba 1 1 1 1>)
    ScaleA=true
    ScaleRGB=true
   )
   &ScaleRibbonWidth(Ribbon Width Scale=&FloatFromCurve(FloatCurve=DataInterface() {propertyValues {Curve {keys [
         {arriveTangent=-0.0755, interpMode=RCIM_Cubic, leaveTangent=-0.0755, tangentMode=RCTM_User, time=0, value=1}
         {arriveTangent=-1.6226, interpMode=RCIM_Cubic, leaveTangent=-1.6226, tangentMode=RCTM_User, time=1, value=0.3}
        ]}}}))
  ]
  rendererClasses [&NiagaraRibbonRendererProperties]
  renderers [{rendererClass=&NiagaraRibbonRendererProperties}]
 }
}
summary {
 systemName=NS_Pickaxe_OceanBreeze_Trail
 emitters [
  {emitterName=Empty, rendererClasses [&NiagaraSpriteRendererProperties]}
  {emitterName=FN_AnimTrail_Short, rendererClasses [&NiagaraRibbonRendererProperties]}
  {emitterName=FN_AnimTrail_Short004, rendererClasses [&NiagaraRibbonRendererProperties]}
  {emitterName=FN_AnimTrail_Short005, rendererClasses [&NiagaraRibbonRendererProperties]}
  {emitterName=FN_AnimTrail_Short006, rendererClasses [&NiagaraRibbonRendererProperties]}
  {emitterName=Trail_color, rendererClasses [&NiagaraRibbonRendererProperties]}
 ]
 userVariables [
  User.bIsRightHand {type=NiagaraBool}
  User.Color A {type=LinearColor, description="", defaultValue {struct=LinearColor, value=<rgba 1 1 1 1>}}
  User.ColorSwitcher {type=NiagaraFloat, defaultValue=0, description=""}
  User.RibbonIDOffset {type=NiagaraInt32}
  User.RibbonWidth {type=NiagaraFloat, defaultValue=10, description=""}
  User.SkeletalMesh {
   type=NiagaraDataInterfaceSkeletalMesh
   defaultValue {
    struct=NiagaraExt_VariableValue_DataInterface
    value {dataInterface=&self:NiagaraDataInterfaceSkeletalMesh_0}
   }
  }
  User.Swing_SK_Mesh {
   type=NiagaraDataInterfaceSkeletalMesh
   defaultValue {
    struct=NiagaraExt_VariableValue_DataInterface
    value {dataInterface=&self:NiagaraDataInterfaceSkeletalMesh_1}
   }
  }
 ]
}
